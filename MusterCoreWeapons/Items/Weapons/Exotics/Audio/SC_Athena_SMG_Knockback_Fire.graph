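# Third-person fire sound cue for an exotic Athena knockback SMG, built from Pistol PDW gunshot waves.
# A mixer blends close, medium and distant layers, each with its own attenuation and a random pick of three waves.
# Medium and distant layers share the same distant recordings, while the close layer is pitched up and the distant one boosted before mixing.
# Sound class overrides separate distant from nearer layers, and local and global gunfire concurrency limits apply.
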SoundCue /MusterCoreWeapons/Items/Weapons/Exotics/Audio/SC_Athena_SMG_Knockback_Fire
root=Mixer_17
nodes [
 {
  id=Attenuation_0
  children [Modulator_2]
  properties {attenuationSettings=&Fort_Weapon_Projectile_Medium_Mono_Athena_Att}
 }
 {
  id=Attenuation_34
  children [Modulator_0]
  properties {attenuationSettings=&Fort_Weapon_Projectile_Distant_Mono_Athena_Att}
 }
 {id=Attenuation_49, children [Modulator_1], properties {attenuationSettings=&Fort_Weapon_Projectile_Close_Athena_Att}}
 {id=Mixer_17, children [SoundClass_32, SoundClass_0, SoundClass_35], properties {inputVolume [0.75,0.4,0.1]}}
 {id=Modulator_0, children [Random_2], properties {pitchMax=1, pitchMin=1, volumeMax=2, volumeMin=2}}
 {id=Modulator_1, children [Random_0], properties {pitchMax=1.1, pitchMin=1.1}}
 {id=Modulator_2, children [Random_1], properties {pitchMax=1, pitchMin=1, volumeMax=1, volumeMin=1}}
 {id=Random_0, children [WavePlayer_0,WavePlayer_1,WavePlayer_2], properties {weights [1,1,1]}}
 {id=Random_1, children [WavePlayer_3,WavePlayer_4,WavePlayer_5], properties {weights [1,1,1]}}
 {id=Random_2, children [WavePlayer_9, WavePlayer_10, WavePlayer_11], properties {weights [1,1,1]}}
 {
  id=SoundClass_0
  children [Attenuation_0]
  properties {soundClassOverride=&Fort_WeaponRanged_Fire_3P_PSM}
 }
 {
  id=SoundClass_32
  children [Attenuation_49]
  properties {soundClassOverride=&Fort_WeaponRanged_Fire_3P_PSM}
 }
 {
  id=SoundClass_35
  children [Attenuation_34]
  properties {soundClassOverride=&Fort_WeaponRanged_Fire_Distant}
 }
 {id=WavePlayer_0, properties {soundWaveAssetPtr=/MusterCoreWeapons/Items/Weapons/Exotics/Audio/SoundWavs/Pistol_PDW_Fire_3p_Close_01}}
 {id=WavePlayer_1, properties {soundWaveAssetPtr=/MusterCoreWeapons/Items/Weapons/Exotics/Audio/SoundWavs/Pistol_PDW_Fire_3p_Close_02}}
 {id=WavePlayer_10, properties {soundWaveAssetPtr=/MusterCoreWeapons/Items/Weapons/Exotics/Audio/SoundWavs/Pistol_PDW_Fire_3p_Dist_02}}
 {id=WavePlayer_11, properties {soundWaveAssetPtr=/MusterCoreWeapons/Items/Weapons/Exotics/Audio/SoundWavs/Pistol_PDW_Fire_3p_Dist_03}}
 {id=WavePlayer_2, properties {soundWaveAssetPtr=/MusterCoreWeapons/Items/Weapons/Exotics/Audio/SoundWavs/Pistol_PDW_Fire_3p_Close_03}}
 {id=WavePlayer_3, properties {soundWaveAssetPtr=/MusterCoreWeapons/Items/Weapons/Exotics/Audio/SoundWavs/Pistol_PDW_Fire_3p_Dist_01}}
 {id=WavePlayer_4, properties {soundWaveAssetPtr=&Pistol_PDW_Fire_3p_Dist_02}}
 {id=WavePlayer_5, properties {soundWaveAssetPtr=&Pistol_PDW_Fire_3p_Dist_03}}
 {id=WavePlayer_9, properties {soundWaveAssetPtr=&Pistol_PDW_Fire_3p_Dist_01}}
]
properties {
 bDebug=true
 maxDistance=27500
 volumeMultiplier=1.07
 concurrencySet [&Weapon_Gunfire_Local_Conc, &Weapon_Gunfire_Global_Conc]
}
