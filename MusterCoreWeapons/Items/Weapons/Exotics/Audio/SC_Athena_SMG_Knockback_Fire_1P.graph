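# First-person firing sound cue for the knockback SMG, an exotic weapon in the Athena mode.
# A random node picks one of three equally weighted PDW pistol close-fire waves, reusing third-person recordings, and a modulator shifts it to a fixed raised pitch.
# It boosts volume and uses first-person weapon attenuation, the ranged-fire sound class and the first-person gunfire concurrency group.
# The debug flag is left on.
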SoundCue /MusterCoreWeapons/Items/Weapons/Exotics/Audio/SC_Athena_SMG_Knockback_Fire_1P
root=Modulator_1
nodes [
 {id=Modulator_1, children [Random_0], properties {pitchMax=1.1, pitchMin=1.1}}
 {id=Random_0, children [WavePlayer_0,WavePlayer_1,WavePlayer_2], properties {weights [1,1,1]}}
 {id=WavePlayer_0, properties {soundWaveAssetPtr=/MusterCoreWeapons/Items/Weapons/Exotics/Audio/SoundWavs/Pistol_PDW_Fire_3p_Close_01}}
 {id=WavePlayer_1, properties {soundWaveAssetPtr=/MusterCoreWeapons/Items/Weapons/Exotics/Audio/SoundWavs/Pistol_PDW_Fire_3p_Close_02}}
 {id=WavePlayer_2, properties {soundWaveAssetPtr=/MusterCoreWeapons/Items/Weapons/Exotics/Audio/SoundWavs/Pistol_PDW_Fire_3p_Close_03}}
]
properties {
 attenuationSettings=&Fort_Weapon_Projectile_FirstPerson_Athena_Att
 bDebug=true
 bOverrideConcurrency=true
 maxDistance=2097152
 soundClassObject=&Fort_WeaponRanged_Fire_1P_PSM
 volumeMultiplier=1.3
 concurrencySet [&Weapon_Gunfire_FirstPerson_Conc]
}
