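# Sound cue for the foley of leaving aim-down-sights on an exotic SMG, named Slap, in MusterCoreWeapons.
# A Random node picks between two equally weighted FirePetal ADS-out waves and feeds a Modulator at the root.
# It plays at a heavily reduced volume in the weapon foley sound class, using the shared weapon reload attenuation.
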SoundCue /MusterCoreWeapons/Items/Weapons/Exotics/Audio/SC_Athena_SMG_Slap_ADS_OUT
root=Modulator_0
nodes [
 {id=Modulator_0, children [Random_0]}
 {id=Random_0, children [WavePlayer_0,WavePlayer_3], properties {weights [1,1]}}
 {id=WavePlayer_0, properties {soundWaveAssetPtr=/MusterCoreWeapons/Items/Weapons/Exotics/Audio/SoundWavs/SMG_FirePetal_ADS_Out_02}}
 {id=WavePlayer_3, properties {soundWaveAssetPtr=/MusterCoreWeapons/Items/Weapons/Exotics/Audio/SoundWavs/SMG_FirePetal_ADS_Out_01}}
]
properties {
 attenuationSettings=&Fort_Weapon_Reload_Att
 maxDistance=2250
 soundClassObject=&Fort_WeaponFoley
 volumeMultiplier=0.1
}
