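# Sound cue for the drop sound of the Athena exotic "Slap" SMG, reusing pistol PDW drop recordings.
# A Random node picks one of three equally weighted waves and feeds a Modulator at the root.
# It uses shared weapon-drop attenuation, the loot-collect sound class and a stop-oldest weapon-drop concurrency set, with a raised volume multiplier.
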SoundCue /MusterCoreWeapons/Items/Weapons/Exotics/Audio/SC_Athena_SMG_Slap_Drop
root=Modulator_0
nodes [
 {id=Modulator_0, children [Random_2]}
 {id=Random_2, children [WavePlayer_4,WavePlayer_5,WavePlayer_6], properties {weights [1,1,1]}}
 {id=WavePlayer_4, properties {soundWaveAssetPtr=/MusterCoreWeapons/Items/Weapons/Exotics/Audio/SoundWavs/Pistol_PDW_Drop_01}}
 {id=WavePlayer_5, properties {soundWaveAssetPtr=/MusterCoreWeapons/Items/Weapons/Exotics/Audio/SoundWavs/Pistol_PDW_Drop_02}}
 {id=WavePlayer_6, properties {soundWaveAssetPtr=/MusterCoreWeapons/Items/Weapons/Exotics/Audio/SoundWavs/Pistol_PDW_Drop_03}}
]
properties {
 attenuationSettings=&Fort_Weapon_Drop_Att
 maxDistance=3000
 soundClassObject=&Fort_Loot_Collect
 volumeMultiplier=1.25
 concurrencyOverrides {resolutionRule=PreventNew}
 concurrencySet [&Weapon_Drop_StopOldest]
}
