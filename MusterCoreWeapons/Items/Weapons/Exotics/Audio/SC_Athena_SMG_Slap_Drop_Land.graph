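# Sound cue for the impact heard when the exotic Slap SMG lands after being dropped.
# A Random node picks one of three equally weighted Pistol/PDW drop impact waves and feeds it through a Modulator for variation.
# It uses the shared weapon-drop attenuation, the loot-collect sound class and the Weapon_Drop_StopOldest concurrency set.
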SoundCue /MusterCoreWeapons/Items/Weapons/Exotics/Audio/SC_Athena_SMG_Slap_Drop_Land
root=Modulator_0
nodes [
 {id=Modulator_0, children [Random_0]}
 {id=Random_0, children [WavePlayer_7,WavePlayer_8,WavePlayer_9], properties {weights [1,1,1]}}
 {id=WavePlayer_7, properties {soundWaveAssetPtr=/MusterCoreWeapons/Items/Weapons/Exotics/Audio/SoundWavs/Pistol_PDW_Drop_Impact_01}}
 {id=WavePlayer_8, properties {soundWaveAssetPtr=/MusterCoreWeapons/Items/Weapons/Exotics/Audio/SoundWavs/Pistol_PDW_Drop_Impact_02}}
 {id=WavePlayer_9, properties {soundWaveAssetPtr=/MusterCoreWeapons/Items/Weapons/Exotics/Audio/SoundWavs/Pistol_PDW_Drop_Impact_03}}
]
properties {
 attenuationSettings=&Fort_Weapon_Drop_Att
 maxDistance=3000
 soundClassObject=&Fort_Loot_Collect
 volumeMultiplier=1
 concurrencyOverrides {resolutionRule=PreventNew}
 concurrencySet [&Weapon_Drop_StopOldest]
}
