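# Sound cue for the end of the slap reload on an exotic Athena SMG, built from PDW pistol reload foley recordings.
# A Random node picks one of three equally weighted reload-end waves and feeds it through a default Modulator for slight variation.
# It plays quietly in the weapon foley class with the shared weapon reload attenuation, and overrides concurrency to prevent new instances once the limit is reached.
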SoundCue /MusterCoreWeapons/Items/Weapons/Exotics/Audio/SC_Athena_SMG_Slap_Reload_End
root=Modulator_1
nodes [
 {id=Modulator_1, children [Random_0]}
 {id=Random_0, children [WavePlayer_0,WavePlayer_2,WavePlayer_3], properties {weights [1,1,1]}}
 {id=WavePlayer_0, properties {soundWaveAssetPtr=/MusterCoreWeapons/Items/Weapons/Exotics/Audio/SoundWavs/Pistol_PDW_Reload_End_01}}
 {id=WavePlayer_2, properties {soundWaveAssetPtr=/MusterCoreWeapons/Items/Weapons/Exotics/Audio/SoundWavs/Pistol_PDW_Reload_End_02}}
 {id=WavePlayer_3, properties {soundWaveAssetPtr=/MusterCoreWeapons/Items/Weapons/Exotics/Audio/SoundWavs/Pistol_PDW_Reload_End_03}}
]
properties {
 attenuationSettings=&Fort_Weapon_Reload_Att
 bOverrideConcurrency=true
 maxDistance=2250
 soundClassObject=&Fort_WeaponFoley
 volumeMultiplier=0.35
 attenuationOverrides {falloffDistance=4100, attenuationShapeExtents {x=900}}
 concurrencyOverrides {resolutionRule=PreventNew}
}
