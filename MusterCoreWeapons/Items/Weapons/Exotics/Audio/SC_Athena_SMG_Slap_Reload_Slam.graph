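# Sound cue for the slap/slam step of an exotic SMG reload, played as weapon foley.
# A random node picks one of three equally weighted PDW reload slap waves and feeds it through a modulator left at default settings for variation.
# It plays at reduced volume with the shared weapon reload attenuation, and its concurrency override blocks new instances once the limit is reached.
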SoundCue /MusterCoreWeapons/Items/Weapons/Exotics/Audio/SC_Athena_SMG_Slap_Reload_Slam
root=Modulator_2
nodes [
 {id=Modulator_2, children [Random_1]}
 {id=Random_1, children [WavePlayer_6,WavePlayer_7,WavePlayer_8], properties {weights [1,1,1]}}
 {id=WavePlayer_6, properties {soundWaveAssetPtr=/MusterCoreWeapons/Items/Weapons/Exotics/Audio/SoundWavs/Pistol_PDW_Reload_Slap_01}}
 {id=WavePlayer_7, properties {soundWaveAssetPtr=/MusterCoreWeapons/Items/Weapons/Exotics/Audio/SoundWavs/Pistol_PDW_Reload_Slap_02}}
 {id=WavePlayer_8, properties {soundWaveAssetPtr=/MusterCoreWeapons/Items/Weapons/Exotics/Audio/SoundWavs/Pistol_PDW_Reload_Slap_03}}
]
properties {
 attenuationSettings=&Fort_Weapon_Reload_Att
 bOverrideConcurrency=true
 maxDistance=2250
 soundClassObject=&Fort_WeaponFoley
 volumeMultiplier=0.35
 attenuationOverrides {falloffDistance=4100, attenuationShapeExtents {x=900}}
 concurrencyOverrides {resolutionRule=PreventNew}
}
