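# Sound cue for an exotic energy weapon's activation, playing an Energy Bottle start sound at reduced volume under the gadgets and items sound class.
# An IsLocal branch picks stereo status-effect attenuation when true, and the mono version when false or unset.
# Both paths play the same wave through separate players, so only the attenuation differs between local and remote listeners.
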SoundCue /MusterCoreWeapons/Items/Weapons/Exotics/Audio/SC_Energy_Weapon_Activated
root=Branch_2
nodes [
 {
  id=Attenuation_2
  children [WavePlayer_0]
  properties {attenuationSettings=/FNGameplayCues/Sounds/Attenuation/Fort_Ability_Status_Effect_Looping_Stereo_Att}
 }
 {
  id=Attenuation_3
  children [WavePlayer_5]
  properties {attenuationSettings=/FNGameplayCues/Sounds/Attenuation/Fort_Ability_Status_Effect_Looping_Mono_Att}
 }
 {id=Branch_2, children [Attenuation_2, Attenuation_3, Attenuation_3], properties {boolParameterName=IsLocal}}
 {id=WavePlayer_0, properties {soundWaveAssetPtr=/Energy/Sounds/Energy_Bottle_EnergyStart_01}}
 {id=WavePlayer_5, properties {soundWaveAssetPtr=&Energy_Bottle_EnergyStart_01}}
]
properties {maxDistance=3500, soundClassObject=&Fort_Gadgets_Items, volumeMultiplier=0.3}
