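; Explosive AR bullet projectile that explodes on impact, with a timed fallback explosion for long-lived bouncing shots.
; A fake tracer mesh animates through material parameters fed by muzzle location, start time and speed.
; Blocking touches record hit data on authority, and stopping spawns explosion FX unless the water FX flag is set.
; Water entry sets that flag, traces from the weapon to find the surface, plays splash effects and destroys the projectile.
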(event EventBeginPlay
  (bind _tracermesh (Variables|Default|GetTracerMesh))
  ;[1 1 1] EVENT -> BeginPlay
  ;;[1 1 1] With bouncy surfaces, there's a possibility for a long life. Explode after a time
  (Utilities|Time|SetTimerbyEvent (EventDispatchers|CreateEvent :!event DoExplosion) 22.0)
  ;[1 1 1] Animate fake tracer mesh
  (Networking|CanExecuteCosmeticEvents
    (:True
      (bind _asfort_player_pawn (Utilities|Casting|CastToFortPlayerPawn (Variables|Actor|GetInstigator)))
      (Variables|Default|SetCurrent%20Weapon (Variables|/Script/FortniteGame.FortPawn|GetCurrentWeapon _asfort_player_pawn))
      (Rendering|Material|SetVectorParameterValueonMaterials _tracermesh "BulletOrigin" (Effects|GetMuzzleLocation (Variables|/Script/FortniteGame.FortPawn|GetCurrentWeapon _asfort_player_pawn) -1))
      (Rendering|Material|SetScalarParameterValueonMaterials _tracermesh "StartTime" (Utilities|Time|GetGameTimeinSeconds))
      (Rendering|Material|SetScalarParameterValueonMaterials _tracermesh "WPODuration" (Math|Float|MapRangeClamped (Math|Vector|VectorLength (Transformation|GetVelocity)) 5e+03 4e+04 0.1 0.25)))))

(event EventOnTouched
  (params (OtherActor :type Actor%20Object%20Reference) (OtherComp :type Primitive%20Component%20Object%20Reference) (HitResult :type Hit%20Result%20Structure%20%28by%20ref%29) (bIsOverlap :type Boolean))
  (bind (_bblockinghit _binitialoverlap _time _distance _location _impactpoint _normal _impactnormal _physmat _hitactor) (Collision|BreakHitResult HitResult))
  (bind (_asfort_pawn _bsuccess) (Utilities|Casting|CastToFortPawn _hitactor))
  ;[1 1 1] EVENT -> OnTouched
  (if bIsOverlap
    (Utilities|IsValid _physmat
      (:Is%20Valid
        (if (Utilities|Enum|Equal(Enum) (Variables|/Script/PhysicsCore.PhysicalMaterial|GetSurfaceType _physmat) "SurfaceType8")
          (Audio|PlaySoundatLocation (Variables|Default|GetSound) _location))))
    (else
      (Networking|SwitchHasAuthority
        (:Authority
          (if (not (Delivery|IsBeingKilled))
            (CallFunction|SetHitLocation _location)
            (CallFunction|SetHitRotation (Math|Rotator|MakeRotfromX _normal))
            (CallFunction|SetHitPlayer _bsuccess)))))))

(event OnEnterWaterEvent(WaterInteractionComponent)
  (params (WaterBody :type Fort%20Water%20Body%20Actor%20Object%20Reference) (WaterInteractionComponent :type Fort%20Water%20Interaction%20Component%20Object%20Reference) (bIsFirstBody :type Boolean))
  (bind _returnvalue (Transformation|GetActorLocation))
  (bind _current_weapon (Variables|Default|GetCurrent%20Weapon))
  (bind _returnvalue_2 (Transformation|GetActorLocation _current_weapon))
  (bind _sound (Variables|Default|GetSound))
  ; WATER INTERACTION JIRA #FORT-1055999
  (Networking|CanExecuteCosmeticEvents
    (:True
      (if bIsFirstBody
        (Variables|Default|SetbIsWaterFXSpawned true)
        (Utilities|IsValid WaterBody
          (:Is%20Valid
            (bind (_outhit _returnvalue_3) (Collision|LineTraceForObjects _returnvalue_2 _returnvalue (Utilities|Array|MakeArray "ObjectTypeQuery13")))
            (Rendering|Debug|DrawDebugLine _returnvalue_2 _returnvalue)
            (bind _outwatersurfaceinfo (WaterBody|GetWaterSurfaceInfoStruct WaterBody (select _returnvalue_3 (.ImpactPoint (Collision|BreakHitResult _outhit)) (Transformation|GetActorLocation)) true true false))
            (Audio|PlaySoundatLocation _sound (Utilities|Struct|BreakWaterSurfaceInfo _outwatersurfaceinfo))
            (Utilities|FlowControl|Sequence
              (:then_0
                (Utilities|IsValid _sound
                  (:Is%20Valid
                    (Utilities|FlowControl|DoOnce false
                      (:Completed
                        (Audio|PlaySoundatLocation _sound (Utilities|Struct|BreakWaterSurfaceInfo _outwatersurfaceinfo))
                        (bind _returnvalue_4 (Effects|Components|ParticleSystem|SpawnAltFXAtLocation "WeaponImpactEffect" (Variables|Default|GetNS%20Water%20Explosion%20FX) 0 (Utilities|Struct|BreakWaterSurfaceInfo _outwatersurfaceinfo) (Math|Rotator|MakeRotfromX (.WaterPlaneNormal (Utilities|Struct|BreakWaterSurfaceInfo _outwatersurfaceinfo)))))
                        (Effects|Components|ParticleSystem|SetFloatParameter _returnvalue_4 "WaterDepth" 0.0)
                        (Effects|Components|ParticleSystem|SetFloatParameter _returnvalue_4 "WaterBodyIndex" (Math|Conversions|ToFloat(Integer) 0))
                        (Effects|Components|ParticleSystem|SetFloatParameter _returnvalue_4 "Global Scale" (Variables|Default|GetExplosion%20Global%20Scale))
                        (Actor|DestroyActor)
                        (Utilities|FlowControl|Delay 0.5)
                        (Variables|Default|SetbIsWaterFXSpawned false)))))))))))))

(event EventOnStop
  (params (Hit :type Hit%20Result%20Structure%20%28by%20ref%29))
  (bind (_bblockinghit _binitialoverlap _time _distance _location _impactpoint _normal _impactnormal) (Collision|BreakHitResult Hit))
  ; ON STOP JIRA #FORT-1055999
  (|CallParentFunction Hit)
  (if (not (Variables|Default|GetbIsWaterFXSpawned))
    (Effects|Components|ParticleSystem|SpawnAltFXAtLocation "WeaponImpactEffect" (Variables|Default|GetNS%20Explosion%20FX) 0 _location (Math|Rotator|MakeRotfromX _impactnormal))))

(fn ConstructionScript
  (LOD|SetMaxDrawDistance (Variables|Default|GetTracerMesh) (Variables|FortMusterExplosiveARProjectile|GetDrawDistance)))
