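; Construction-script Blueprint for the exotic heavy breach shotgun weapon in Fortnite's Athena mode.
; After calling the parent construction script, it sets the weapon's trace-through-buildings limit.
; The limit comes from a scalable float variable evaluated at level 1 and truncated to an integer, so it is data-driven.
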(fn ConstructionScript
  (|CallParentFunction)
  (FortWeapon|SetTraceThroughBuildingsLimit (Math|Float|Truncate (ScalableFloat|GetValueatLevel (Variables|Default|GetRow_TraceThroughBuildingsLimit) 1.0))))
