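; Gameplay ability for an exotic SMG that knocks its owner back opposite their aim, gated by an avatar tag check.
; On the server, LaunchPlayer skips vehicle or knockback-immune pawns and gives grounded crouchers a small backward push.
; Otherwise it applies a stacking effect whose count scales horizontal launch from pitch-offset aim, optionally resetting fall height on upward launches.
; A delayed velocity check fires a gameplay cue on high vertical speed, then ends the ability.
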(event Ability|EventActivateAbility
  (bind _returnvalue (GameplayTags|HasAnyMatchingGameplayTags(Message) (Ability|GetAvatarActorfromActorInfo) (Variables|Default|GetTag%20Container)))
  (if _returnvalue
    (Development|PrintString (Utilities|String|ToString(Boolean) false))
    ;[1 1 1] Wait a brief moment (Like 2 frames) then check velocity. If Velocity on Z is high enough or low enough play GC.
    (CallFunction|LaunchPlayer)
    (Utilities|Time|SetTimerbyEvent (delegate Custom|CheckVel) (Variables|Default|GetFX%20Delay))))

(event Custom|CheckVel
  (bind _returnvalue (Ability|GetAvatarActorfromActorInfo))
  ;[1 1 1] Wait a brief moment (Like 2 frames) then check velocity. If Velocity on Z is high enough or low enough play GC.
  (Utilities|IsValid _returnvalue
    (:Is%20Valid
      (bind _asfort_player_pawn (Utilities|Casting|CastToFortPlayerPawn _returnvalue))
      (Development|PrintString (Utilities|String|ToString(Vector) (Transformation|GetVelocity _asfort_player_pawn)))
      (bind (_returnvalue_x _returnvalue_y _returnvalue_z) (Math|Vector|BreakVector (Transformation|GetVelocity _asfort_player_pawn)))
      (if (> (Math|Float|Absolute(Float) _returnvalue_z) (Variables|Default|GetVelocityForFX))
        (Ability|ExecuteGameplayCueOnOwner (Variables|Default|GetKnockbackCue))
        (else
          _))
      (Ability|EndAbility))))

(event Ability|EventFailedToActivatePassiveAbility
  (params (ActorInfo :type Gameplay%20Ability%20Actor%20Info%20Structure))
  (|CallParentFunction))

(fn IsInVehicle
  (params (FortPawn :type Fort%20Pawn%20Object%20Reference))
  (bind _asfort_player_pawn (Utilities|Casting|CastToFortPlayerPawn FortPawn)
    (:then
      (return (Vehicle|IsinVehicle _asfort_player_pawn)))
    (:CastFailed
      (return false))))

(fn LaunchPlayer
  (bind _returnvalue (Ability|GetAvatarActorfromActorInfo))
  (bind _finallaunchvel (Variables|Default|GetFinalLaunchVel))
  ;[1 1 1] Cast to pawn and launch the player a little bit. The distance launcged is based on a max distance from the player who hit them. If the player is further than the max distance than no launch is applied. Otherwise the player is launched based on a ratio of the knock mag using the distance from the player who launched them.
  (if (Networking|IsServer)
    (Utilities|IsValid _returnvalue
      (:Is%20Valid
        (bind _asfort_pawn (Utilities|Casting|CastToFortPawn _returnvalue))
        ;;[1 1 1] Check for Knock Immune or In Vehicle
        (if (not (or (CallFunction|IsinVehicle _asfort_pawn) (GameplayTags|HasMatchingGameplayTag _asfort_pawn "(TagName=\"Gameplay.Status.KnockbackImmunity\")")))
          ;;[1 1 1] If Crouched do different knock, but not if they are falling, if they are falling do the normal one.
          (if (and (Utilities|Enum|NotEqual(Enum) (Variables|/Script/Engine.CharacterMovementComponent|GetMovementMode (Variables|/Script/Engine.Character|GetCharacterMovement _asfort_pawn)) "MOVE_Falling") (Variables|/Script/Engine.Character|GetbIsCrouched _asfort_pawn))
            ;;[1 1 1] Launch the player down a little to make the ground splash.
            (Character|LaunchCharacter _asfort_pawn (Math|Vector|MakeVector (.x (* (Transformation|GetActorForwardVector _asfort_pawn) -1e+02)) (.y (* (Transformation|GetActorForwardVector _asfort_pawn) -1e+02))) true true)
            (else
              (bind _asfort_player_pawn (Utilities|Casting|CastToFortPlayerPawn _asfort_pawn)
                (:then
                  (FortMovement|CancelJumpAbility _asfort_player_pawn false))
                (:CastFailed
                  _))
              (bind _returnvalue_2 (Ability|ApplyGameplayEffectToOwner "/MusterCoreWeapons/Items/Weapons/Exotics/GE_smg_X_KnockbackSelf_stacking.GE_smg_X_KnockbackSelf_stacking_C"))
              (bind _returnvalue_3 (Ability|GameplayEffect|GetActiveGameplayEffectStackCount _returnvalue_2))
              (bind _output_get (LocalVariables|SetStackcount _returnvalue_3))
              (Development|PrintString (Utilities|String|ToString(Integer) _output_get) true true "(R=1.000000,G=0.000000,B=0.026683,A=1.000000)")
              (Variables|Default|SetMultiplier (* (+ (/ _output_get 10) 1) 2))
              ;;[1 1 1] Calc Launch Vel
              (bind _output_get_2 (Variables|Default|SetFinalLaunchVel (Math|Vector|MakeVector (* (.x (* (* (Math|Vector|Normalize (Math|Vector|GetForwardVector (Math|Rotator|MakeRotator (.roll (Pawn|GetBaseAimRotation _asfort_pawn)) (+ (.pitch (Pawn|GetBaseAimRotation _asfort_pawn)) (ScalableFloat|GetValueatLevel (Variables|Default|GetRow_OffsetPitch))) (.yaw (Pawn|GetBaseAimRotation _asfort_pawn))))) (ScalableFloat|GetValueatLevel (Variables|Default|GetRow_KnockMag))) -1.0)) (Variables|Default|GetMultiplier)) (* (.y (* (* (Math|Vector|Normalize (Math|Vector|GetForwardVector (Math|Rotator|MakeRotator (.roll (Pawn|GetBaseAimRotation _asfort_pawn)) (+ (.pitch (Pawn|GetBaseAimRotation _asfort_pawn)) (ScalableFloat|GetValueatLevel (Variables|Default|GetRow_OffsetPitch))) (.yaw (Pawn|GetBaseAimRotation _asfort_pawn))))) (ScalableFloat|GetValueatLevel (Variables|Default|GetRow_KnockMag))) -1.0)) (Variables|Default|GetMultiplier)) (Math|Float|float+float (.z (* (* (Math|Vector|Normalize (Math|Vector|GetForwardVector (Math|Rotator|MakeRotator (.roll (Pawn|GetBaseAimRotation _asfort_pawn)) (+ (.pitch (Pawn|GetBaseAimRotation _asfort_pawn)) (ScalableFloat|GetValueatLevel (Variables|Default|GetRow_OffsetPitch))) (.yaw (Pawn|GetBaseAimRotation _asfort_pawn))))) (ScalableFloat|GetValueatLevel (Variables|Default|GetRow_KnockMag))) -1.0)) (ScalableFloat|GetValueatLevel (Variables|Default|GetRow_ZAdd)) (* (.z (Transformation|GetVelocity _asfort_pawn)) (ScalableFloat|GetValueatLevel (Variables|Default|GetRow_ZVelMult)))))))
              (Development|PrintString (Utilities|String|ToString(Vector) _output_get_2))
              ;;[1 1 1] Stop Movement and launch player
              (Components|Movement|StopMovementImmediately (Variables|/Script/Engine.Character|GetCharacterMovement _asfort_pawn))
              (Character|LaunchCharacter _asfort_pawn _finallaunchvel true true)
              ;;[1 1 1] If Velocity is Pos Z reset fall height
              (if (and (Math|Conversions|ToBoolean(Integer) (Math|Float|Floor (ScalableFloat|GetValueatLevel (Variables|Default|GetRow_ResetFall)))) (> (.z _finallaunchvel) 0.0))
                (Pawn|Character|ResetFallingHeight _asfort_pawn)
                (else
                  _)))))))))
